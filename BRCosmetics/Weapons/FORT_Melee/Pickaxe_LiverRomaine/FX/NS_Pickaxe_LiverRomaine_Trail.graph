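# Niagara swing trail effect for the LiverRomaine pickaxe, combining two ribbon trails with shed envelope mesh particles.
# TrailBurgundy and TrailToon build short-lived anim-trail ribbons with fading alpha and system time-of-day tint; TrailToon adds a second renderer and fading toon-line color.
# Envelopes spawns by distance moved at socket locations, varying color and scale by partition.
# Envelopes inherit some velocity, then drag, fall and shrink, using toon shader lighting parameters.
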NiagaraSystem /BRCosmetics/Weapons/FORT_Melee/Pickaxe_LiverRomaine/FX/NS_Pickaxe_LiverRomaine_Trail
dependencies {
 usedDataInterfaces [&NiagaraDataInterfaceCurve, &NiagaraDataInterfaceSocketReader, &NiagaraDataInterfaceColorCurve]
 usedDynamicInputs [&ColorFromCurve]
 usedModules [
  &self:SystemSpawnScript.SystemScriptSource.SystemScriptGraph.NiagaraNodeAssignment_0.SetVariables_#1
  &self:SystemSpawnScript.SystemScriptSource.SystemScriptGraph.NiagaraNodeAssignment_2.SetVariables_#2
  &SystemState
 ]
}
emitters {
 Envelopes {
  emitterSpawnScript [/SetVariables(
    Emitter.TOD_Lit:Vector3f=&MakeVectorFromLinearColorRGB(Color=&FNTimeOfDayColor(
      Night Color=<rgba 0.156 0.1907 0.26 1>
      Morning Color=<rgba 0.6 0.6 0.6 1>
      Day Color=<rgba 0.5 0.5 0.5 1>
      Evening Color=<rgba 0.5 0.5 0.5 1>
     ))
    Emitter.TOD_Shaded:Vector3f=&MakeVectorFromLinearColorRGB(Color=&FNTimeOfDayColor(
      Night Color=<rgba 0.096 0.1173 0.16 1>
      Morning Color=<rgba 0.192 0.192 0.192 1>
      Day Color=<rgba 0.2918 0.2918 0.2918 1>
      Evening Color=<rgba 0.2789 0.2789 0.2789 1>
     ))
   ) {isSetParameters=true}]
  emitterUpdateScript [
   &EmitterState(Life Cycle Mode=::NewEnumerator1, Loop Behavior=::NewEnumerator1, Loop Delay=0.05, Loop Duration=1, UseLoopDelay=true)
   &SpawnPerUnit(Module.Use Max Movement Threshold=true, Module.Use Movement Tolerance=false, Spawn Spacing=100, Use Spawn Probability=true)
   /Game/Effects/Niagara/MasterMaterials/ToonShader/Template/NM_ToonShaderParameters(Lighting_Hardness=1, RimLight_Color:LinearColor=<rgba 1 1 1 1>, RimLight_Width=0)
  ]
  particleSpawnScript [
   &PartitionParticles()
   &InitializeParticle(
    Color=&Multiply_LinearColorByFloat(
     Linear Color=&MergeHelper_ColorSwitch(0=<rgba 1 1 1 1>, 1=<rgba 0.8 0.6249 0.4 1>, 2=?, 3=?, Index=$Particles.PartitionParticles.Partition, Num=::NewEnumerator0)
     Float=&RandomRangeFloat(Minimum=0.8, Maximum=1.2)
    )
    Lifetime Max=1
    Lifetime Min=0.5
    Lifetime Mode=::NewEnumerator1
    Mesh Scale=&MergeHelper_VectorSwitch(0=<1 1.3 0.9>, 1=<1 0.9 1.1>, 2=?, 3=?, Index=$Particles.PartitionParticles.Partition, Num=::NewEnumerator0)
    Mesh Scale Mode=::NewEnumerator4
    UsePositionOffset=false
    Write Material Random=true
   )
   &InitialMeshOrientation(Use Rotation Vector=false)
   &InheritVelocity(Module.Limit Speed=false, Velocity Scale=<0.15 0.15 0.15>)
   &SocketLocation(SamplePosOffset=false, Socket Reader=$System.Socket Reader, Socket Sampling Mode=::NewEnumerator2, VelocitySpeedLimit=false)
   &AddVelocityFromPoint(Module.Use Falloff Distance=false)
  ]
  particleUpdateScript [
   &Drag(
    Drag=&FloatFromCurve(
     FloatCurve=DataInterface() {propertyValues {Curve {
        defaultValue=1
        keys [
         {arriveTangent=2.7559, interpMode=RCIM_Cubic, leaveTangent=2.7559, tangentMode=RCTM_User, time=0, value=0}
         {arriveTangent=0.0224, interpMode=RCIM_Cubic, leaveTangent=0.0224, tangentMode=RCTM_User, time=1, value=1}
        ]
       }}}
     Scale Curve=8
    )
    Rotational Drag=0.1
    Use Linear Drag=true
    Use Rotational Drag=false
   )
   &GravityForce(Gravity=<0 0 -300>)
   &SolveForcesAndVelocity(Clamp Velocity=false, Limit Acceleration=false)
   &ParticleState()
   &ScaleColor(ScaleA=true, ScaleRGB=true)
   &UpdateMeshOrientation(Rotation Vector=&VectorFromFloat(Value=&RandomRangeFloat(Minimum=-1, Maximum=1)))
   &ScaleMeshSize(Scale Factor=&Multiply_VectorByFloat(
     Vector=<1 1 1>
     Float=&FloatFromCurve(
      FloatCurve=DataInterface() {propertyValues {Curve {keys [
          {arriveTangent=4.1785, interpMode=RCIM_Cubic, leaveTangent=4.1785, tangentMode=RCTM_User, time=0, value=0}
          {arriveTangent=0.9995, interpMode=RCIM_Cubic, leaveTangent=-0.1394, tangentMode=RCTM_Break, time=0.1, value=1}
          {arriveTangent=-0.2888, interpMode=RCIM_Cubic, leaveTangent=-0.2888, tangentMode=RCTM_User, time=0.5, value=1}
          {interpMode=RCIM_Cubic, leaveTangent=-1.5153, time=1, value=0}
         ]}}}
      Scale Curve=0.8
     )
    )) {module=ScaleMeshSize001}
  ]
  rendererClasses [&NiagaraMeshRendererProperties]
  renderers [{rendererClass=&NiagaraMeshRendererProperties, rendererIndex=0}]
 }
 TrailBurgundy {
  emitterUpdateScript [
   &EmitterState()
   &FN_SpawnAnimTrail(AnimTrailWidthMode=::FromSecond)
  ]
  particleSpawnScript [
   &FN_InitializeAnimTrailBoneIndecies() {module=FN_InitializeAnimTrailBoneIndecies001}
   &FN_InitializeAnimTrail(AnimTrailWidth=1.1, Color=<rgba 0.15 0.0225 0.0373 1>, Lifetime=0.1) {module=FN_InitializeAnimTrail001}
  ]
  particleUpdateScript [
   &ScaleColor(
    Scale Alpha=&Multiply_Float(
     A=&FloatFromCurve(
      FloatCurve=DataInterface() {propertyValues {Curve {keys [
          {interpMode=RCIM_Cubic, leaveTangent=-1.8709, tangentMode=RCTM_Break, time=0, value=1}
          {arriveTangent=-0.3327, interpMode=RCIM_Cubic, leaveTangent=-3.6795, tangentMode=RCTM_Break, time=0.4, value=0}
         ]}}}
      Scale Curve=5
     )
     B=1
    )
    Scale RGB=&MakeVectorFromLinearColorRGB(Color=$System.TOD)
    ScaleA=true
    ScaleRGB=true
   ) {module=ScaleColor002}
   &FN_UpdateAnimTrail()
   &SolveForcesAndVelocity(Clamp Velocity=false, Limit Acceleration=false)
  ]
  rendererClasses [&NiagaraRibbonRendererProperties]
  renderers [{rendererClass=&NiagaraRibbonRendererProperties}]
 }
 TrailToon {
  emitterUpdateScript [
   &EmitterState()
   &FN_SpawnAnimTrail(AnimTrailWidthMode=::FromSecond)
  ]
  particleSpawnScript [
   &FN_InitializeAnimTrailBoneIndecies() {module=FN_InitializeAnimTrailBoneIndecies001}
   &FN_InitializeAnimTrail(AnimTrailWidth=1.1, Color=<rgba 0.1374 0.135 0.15 1>, Lifetime=0.08) {module=FN_InitializeAnimTrail001}
  ]
  particleUpdateScript [
   &ScaleColor(
    Scale Alpha=&Multiply_Float(
     A=&FloatFromCurve(
      FloatCurve=DataInterface() {propertyValues {Curve {keys [
          {interpMode=RCIM_Cubic, leaveTangent=-1.8709, tangentMode=RCTM_Break, time=0, value=1}
          {arriveTangent=-0.3327, interpMode=RCIM_Cubic, leaveTangent=-3.6795, tangentMode=RCTM_Break, time=0.1, value=0}
         ]}}}
      Scale Curve=3
     )
     B=1
    )
    Scale RGB=&MakeVectorFromLinearColorRGB(Color=$System.TOD)
    ScaleA=true
    ScaleRGB=true
   ) {module=ScaleColor002}
   &FN_UpdateAnimTrail()
   &SolveForcesAndVelocity(Clamp Velocity=false, Limit Acceleration=false)
   /SetVariables(Particles.ToonLinesColor:LinearColor=&MakeLinearColorFromVectorAndFloat(
     "Vector (RGB)"=&MakeVectorFromLinearColorRGB(Color=<rgba 0.2 0.02 0.035 1>)
     "Float (Alpha)"=&Multiply_Float(
      A=&FloatFromCurve(FloatCurve=DataInterface() {propertyValues {Curve {keys [
           {interpMode=RCIM_Cubic, leaveTangent=-1.8709, tangentMode=RCTM_Break, time=0, value=1}
           {arriveTangent=-0.3327, interpMode=RCIM_Cubic, leaveTangent=-3.6795, tangentMode=RCTM_Break, time=1, value=0}
          ]}}})
      B=1
     )
    )) {isSetParameters=true}
  ]
  rendererClasses [&NiagaraRibbonRendererProperties]
  renderers [
   {rendererClass=&NiagaraRibbonRendererProperties}
   {rendererClass=&NiagaraRibbonRendererProperties, rendererIndex=1}
  ]
 }
}
summary {
 systemName=NS_Pickaxe_LiverRomaine_Trail
 emitters [
  {emitterName=TrailBurgundy, rendererClasses [&NiagaraRibbonRendererProperties]}
  {emitterName=TrailToon, rendererClasses [&NiagaraRibbonRendererProperties]}
  {emitterName=Envelopes, rendererClasses [&NiagaraMeshRendererProperties]}
 ]
 userVariables [
  User.bIsFrontEnd {type=NiagaraBool}
  User.bIsFrontEndPreview {type=NiagaraBool}
  User.RibbonIDOffset {type=NiagaraInt32}
 ]
}
