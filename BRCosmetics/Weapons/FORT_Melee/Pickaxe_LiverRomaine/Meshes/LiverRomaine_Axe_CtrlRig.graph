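# Control Rig for the LiverRomaine mailbox pickaxe, posing the mail bones in response to the trapdoor bone's orientation.
# A spherical pose reader on dyn_trapdor outputs one weight, shared by a chain of Modify Transforms nodes on the right, left and under mail groups.
# Each node blends in authored offsets: mostly Y shifts, some rotations, and shrinking of left and under pieces.
# The reader's debug drawing is left enabled.
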ControlRigBlueprint /BRCosmetics/Weapons/FORT_Melee/Pickaxe_LiverRomaine/Meshes/LiverRomaine_Axe_CtrlRig
graphs [
 RigVMFunctionLibrary {links [], nodes []}
 RigVMModel {nodes [
   RigUnit_BeginExecution = &RigUnit_BeginExecution()
   n1 = &RigUnit_SphericalPoseReader(
    ActiveRegionScaleFactors=(PositiveWidth=1,NegativeWidth=1,PositiveHeight=1,NegativeHeight=1)
    ActiveRegionSize=0.15
    Debug=(bDrawDebug=True,bDraw2D=False,bDrawLocalAxes=False,DebugScale=25,DebugSegments=20,DebugThickness=0.25)
    DriverAxis=(X=1,Y=0,Z=0)
    DriverItem=(Type=Bone,Name="dyn_trapdor")
    FalloffRegionScaleFactors=(PositiveWidth=1.8,NegativeWidth=1.8,PositiveHeight=1,NegativeHeight=1)
    FalloffSize=0.3
    FlipHeightScaling=False
    FlipWidthScaling=False
    OptionalParentItem=(Type=Bone,Name="None")
    RotationOffset=(X=90,Y=0,Z=-10.055)
    ExecutePin=@RigUnit_BeginExecution.ExecutePin
   )
   > &RigUnit_ModifyTransforms(
    ItemToModify=((Item=(Type=Bone,Name="dyn_mail_r_MSTR"),Transform=(Rotation=(X=0,Y=0,Z=0.0872,W=0.9962),Translation=(X=0,Y=0,Z=0),Scale3D=(X=1,Y=1,Z=1))))
    WeightMinimum=0
    Weight=@n1.OutputParam
   )
   > &RigUnit_ModifyTransforms(
    ItemToModify=((Item=(Type=Bone,Name="dyn_mail_l_MSTR"),Transform=(Rotation=(X=0,Y=0,Z=0.0872,W=0.9962),Translation=(X=0,Y=-1,Z=0),Scale3D=(X=1,Y=1,Z=1))))
    WeightMinimum=0
    Weight=@n1.OutputParam
   )
   > &RigUnit_ModifyTransforms(
    ItemToModify=((Item=(Type=Bone,Name="dyn_mail_under_MSTR"),Transform=(Rotation=(X=0,Y=0,Z=0,W=1),Translation=(X=0,Y=-5,Z=0),Scale3D=(X=1,Y=1,Z=1))))
    WeightMinimum=0
    Weight=@n1.OutputParam
   )
   > &RigUnit_ModifyTransforms(
    ItemToModify=((Item=(Type=Bone,Name="dyn_mail_under_3"),Transform=(Rotation=(X=0.0614,Y=0.1063,Z=0.2495,W=0.9606),Translation=(X=-5,Y=-10,Z=1),Scale3D=(X=0.8,Y=0.8,Z=0.8))))
    WeightMinimum=0
    Weight=@n1.OutputParam
   )
   > &RigUnit_ModifyTransforms(
    ItemToModify=((Item=(Type=Bone,Name="dyn_mail_under_4"),Transform=(Rotation=(X=0,Y=0,Z=0,W=1),Translation=(X=0,Y=-7,Z=1),Scale3D=(X=0.8,Y=0.8,Z=0.8))))
    WeightMinimum=0
    Weight=@n1.OutputParam
   )
   > &RigUnit_ModifyTransforms(
    ItemToModify=((Item=(Type=Bone,Name="dyn_mail_under_5"),Transform=(Rotation=(X=0,Y=0,Z=0,W=1),Translation=(X=0,Y=-7,Z=1),Scale3D=(X=0.8,Y=0.8,Z=0.8))))
    WeightMinimum=0
    Weight=@n1.OutputParam
   )
   > &RigUnit_ModifyTransforms(
    ItemToModify=((Item=(Type=Bone,Name="dyn_mail_under_2"),Transform=(Rotation=(X=-0.171,Y=-0.171,Z=-0.0302,W=0.9698),Translation=(X=0,Y=-7,Z=2),Scale3D=(X=0.8,Y=0.8,Z=0.8))))
    WeightMinimum=0
    Weight=@n1.OutputParam
   )
   > &RigUnit_ModifyTransforms(
    ItemToModify=((Item=(Type=Bone,Name="dyn_mail_under_1"),Transform=(Rotation=(X=0,Y=0,Z=0,W=1),Translation=(X=0,Y=-7,Z=0),Scale3D=(X=0.8,Y=0.8,Z=0.8))))
    WeightMinimum=0
    Weight=@n1.OutputParam
   )
   > &RigUnit_ModifyTransforms(
    ItemToModify=((Item=(Type=Bone,Name="dyn_mail_l_1"),Transform=(Rotation=(X=0,Y=0,Z=0,W=1),Translation=(X=0,Y=-7,Z=0),Scale3D=(X=0.8,Y=0.8,Z=0.8))))
    WeightMinimum=0
    Weight=@n1.OutputParam
   )
   > &RigUnit_ModifyTransforms(
    ItemToModify=((Item=(Type=Bone,Name="dyn_mail_l_2"),Transform=(Rotation=(X=0,Y=0,Z=0,W=1),Translation=(X=0,Y=-7,Z=0),Scale3D=(X=0.8,Y=0.8,Z=0.8))))
    WeightMinimum=0
    Weight=@n1.OutputParam
   )
   > &RigUnit_ModifyTransforms(
    ItemToModify=((Item=(Type=Bone,Name="dyn_mail_l_3"),Transform=(Rotation=(X=0,Y=0,Z=0,W=1),Translation=(X=0,Y=-7,Z=0),Scale3D=(X=0.8,Y=0.8,Z=0.8))))
    WeightMinimum=0
    Weight=@n1.OutputParam
   )
   > &RigUnit_ModifyTransforms(
    ItemToModify=((Item=(Type=Bone,Name="dyn_mail_r_1"),Transform=(Rotation=(X=0.0842,Y=-0.2578,Z=0.0226,W=0.9623),Translation=(X=0,Y=-7,Z=1),Scale3D=(X=1,Y=1,Z=1))))
    WeightMinimum=0
    Weight=@n1.OutputParam
   )
   > &RigUnit_ModifyTransforms(
    ItemToModify=((Item=(Type=Bone,Name="dyn_mail_r_2"),Transform=(Rotation=(X=0,Y=0,Z=0,W=1),Translation=(X=0,Y=-7,Z=1),Scale3D=(X=1,Y=1,Z=1))))
    WeightMinimum=0
    Weight=@n1.OutputParam
   )
   > &RigUnit_ModifyTransforms(
    ItemToModify=((Item=(Type=Bone,Name="dyn_mail_r_3"),Transform=(Rotation=(X=0,Y=0,Z=0,W=1),Translation=(X=0,Y=-7,Z=1),Scale3D=(X=1,Y=1,Z=1))))
    WeightMinimum=0
    Weight=@n1.OutputParam
   )
   > &RigUnit_ModifyTransforms(
    ItemToModify=((Item=(Type=Bone,Name="dyn_mail_r_4"),Transform=(Rotation=(X=0,Y=0,Z=0,W=1),Translation=(X=0,Y=-9,Z=0),Scale3D=(X=1,Y=1,Z=1))))
    WeightMinimum=0
    Weight=@n1.OutputParam
   )
   > &RigUnit_ModifyTransforms(
    ItemToModify=((Item=(Type=Bone,Name="dyn_mail_r_5"),Transform=(Rotation=(X=0,Y=0,Z=0,W=1),Translation=(X=0,Y=-7,Z=0),Scale3D=(X=1,Y=1,Z=1))))
    WeightMinimum=0
    Weight=@n1.OutputParam
   )
   > &RigUnit_ModifyTransforms(
    ItemToModify=((Item=(Type=Bone,Name="dyn_mail_r_6"),Transform=(Rotation=(X=0,Y=0,Z=0,W=1),Translation=(X=0,Y=-7,Z=0),Scale3D=(X=1,Y=1,Z=1))))
    WeightMinimum=0
    Weight=@n1.OutputParam
   )
  ]}
]
hierarchy [
 b_AR_Root {loc=<0 0 0>, rot=<0 0 0.7071 0.7071>}
 Melee_TwoHanded_Handle {parent=b_AR_Root, loc=<8.1477 -0.008 -29.0793>, rot=<0 0 0.7071 -0.7071>}
 mailbox_MSTR {parent=Melee_TwoHanded_Handle, loc=<0.1364 -0.8766 65.0957>, rot=<0 -0.7071 0 -0.7071>}
 dyn_trapdor {parent=mailbox_MSTR, loc=<0 18.8059 0>, rot=<0 0 0.2139 -0.9769>}
 dyn_mail_r_MSTR {parent=mailbox_MSTR, loc=<-13.5231 -3.3859 -3.8367>, rot=<0 0 0 1>}
 dyn_mail_r_5 {parent=dyn_mail_r_MSTR, loc=<6.6796 15.8744 -0.234>, rot=<0.0458 0.2746 0.0685 -0.958>}
 dyn_mail_r_4 {parent=dyn_mail_r_MSTR, loc=<0.5711 15.6914 2.6575>, rot=<0.1514 0.371 0.194 -0.8955>}
 dyn_mail_r_3 {parent=dyn_mail_r_MSTR, loc=<-0.6872 11.4134 1.2291>, rot=<0.0376 0.4487 0.1295 -0.8835>}
 dyn_mail_r_2 {parent=dyn_mail_r_MSTR, loc=<2.4681 9.971 -4.2834>, rot=<-0.0125 -0.1377 0.1609 -0.9772>}
 dyn_mail_r_1 {parent=dyn_mail_r_MSTR, loc=<-4.2035 13.9028 -0.0032>, rot=<0.0507 0.2868 0.1693 -0.9416>}
 dyn_mail_r_6 {parent=dyn_mail_r_MSTR, loc=<1.1609 13.8024 -1.2598>, rot=<0.1131 0.5144 0.1635 -0.8342>}
 dyn_mail_l_MSTR {parent=mailbox_MSTR, loc=<-13.598 -3.3859 7.0701>, rot=<0 0 0 1>}
 dyn_mail_l_3 {parent=dyn_mail_l_MSTR, loc=<0.9576 17.2193 1.721>, rot=<-0.1168 -0.0914 0.0488 -0.9877>}
 dyn_mail_l_2 {parent=dyn_mail_l_MSTR, loc=<0.2506 21.9371 -0.2979>, rot=<-0.1582 -0.1871 0.1024 -0.9641>}
 dyn_mail_l_1 {parent=dyn_mail_l_MSTR, loc=<-1.4974 19.2259 -1.9607>, rot=<-0.1391 -0.1995 0.1188 -0.9627>}
 dyn_mail_under_5 {parent=dyn_mail_l_MSTR, loc=<6.6293 23.7382 -1.9234>, rot=<0.0309 0.1643 0.1821 -0.969>}
 dyn_mail_under_2 {parent=dyn_mail_l_MSTR, loc=<10.7657 19.7408 -4.2047>, rot=<-0.2266 0.0719 0.2061 -0.9492>}
 dyn_mail_under_1 {parent=dyn_mail_l_MSTR, loc=<7.5966 19.2862 -1.1566>, rot=<0.0029 -0.0602 0.2438 -0.9679>}
 dyn_mail_under_MSTR {parent=mailbox_MSTR, loc=<-4.8257 -3.3859 1.3608>, rot=<0 0 0 1>}
 dyn_mail_under_4 {parent=dyn_mail_under_MSTR, loc=<0.6507 23.3826 -2.6259>, rot=<0.0753 0.0885 0.2216 -0.9682>}
 dyn_mail_under_3 {parent=dyn_mail_under_MSTR, loc=<3.502 25.7457 -6.5963>, rot=<0.158 0.064 0.1935 -0.9662>}
 dyn_mails_back {parent=mailbox_MSTR, loc=<-15.128 -2.5978 0>, rot=<0 0 0 1>}
 FX_Trail_01 {parent=b_AR_Root, type=Null, loc=<-11.0686 -0.008 35.7956>, rot=<0 0 0 1>}
 FX_Trail_02 {parent=b_AR_Root, type=Null, loc=<-3.4865 -0.008 66.334>, rot=<0 0 0 1>}
]
