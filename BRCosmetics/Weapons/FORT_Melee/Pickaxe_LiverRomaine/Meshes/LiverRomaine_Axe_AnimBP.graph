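# Animation Blueprint for the LiverRomaine pickaxe, adding secondary physics to its trapdoor and mail bones.
# The input pose runs through trapdoor AnimDynamics, a Control Rig and additive bone offsets into a cached pose.
# Two differently tuned AnimDynamics mail chains, one world-space, branch from it into a bool blend, with trapdoor-driven limits on some under-mail bones.
# The event graph sets Character_Holding_Pickaxe from BodyMesh validity, capturing BodyMesh when character customization loads.
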AnimBlueprint /BRCosmetics/Weapons/FORT_Melee/Pickaxe_LiverRomaine/Meshes/LiverRomaine_Axe_AnimBP
target_skeleton=&/LiverRomaine_Axe_Skeleton
graphs [
 AnimGraph {nodes [
   Anim Dynamics - Bone: dyn_trapdor(Alpha=1, ExternalForce="0,0,0") {
    angularBiasOverride=0.999
    angularDampingOverride=0.999
    angularSpringConstant=200
    bAngularSpring=true
    bOverrideAngularBias=true
    bOverrideAngularDamping=true
    bUseGravityOverride=true
    simulationSpace=RootRelative
    alphaBoolBlend {blendOption=Linear}
    boundBone {boneName=dyn_trapdor}
    componentAppliedLinearAccClamp=<1000 1000 1000>
    componentLinearAccScale=<0.1 0.1 0.1>
    componentLinearVelScale=<0.1 0.1 0.1>
    gravityOverride {z=-500}
    physicsBodyDefinitions [{
      boundBone {boneName=dyn_trapdor}
      boxExtents=<18 18 18>
      constraintSetup {angularLimitsMax=<0 0 90>, angularLimitsMin=<0 0 -100>, angularTarget=<0 0 0>}
      localJointOffset=<0 -10 0>
     }]
   }
   Anim Dynamics - Bone: dyn_mail_under_MSTR(Alpha=1, ExternalForce="0,0,0") {
    angularBiasOverride=0.999
    angularDampingOverride=0.9
    angularSpringConstant=5000
    bAngularSpring=true
    bOverrideAngularDamping=true
    bUseGravityOverride=true
    alphaBoolBlend {blendOption=Linear}
    boundBone {boneName=dyn_mail_under_MSTR}
    componentAppliedLinearAccClamp=<500 500 500>
    componentLinearAccScale=<0.15 0.15 0.15>
    componentLinearVelScale=<0.15 0.15 0.15>
    gravityOverride {z=-500}
    physicsBodyDefinitions [{
      boundBone {boneName=dyn_mail_under_MSTR}
      boxExtents=<7 7 7>
      constraintSetup {angularLimitsMax=<0 0.5 0.5>, angularLimitsMin=<0 -0.5 -0.5>, angularTarget=<0 0 0>, linearAxesMax=<0 1 0.5>, linearAxesMin=<0 -1 -0.5>}
      localJointOffset=<-1 0 0>
     }]
   }
   Anim Dynamics - Bone: dyn_mail_r_4(Alpha=1, ExternalForce="0,0,0") {
    angularBiasOverride=0.999
    angularDampingOverride=1
    angularSpringConstant=5000
    bAngularSpring=true
    bOverrideAngularBias=true
    bOverrideAngularDamping=true
    bUseGravityOverride=true
    simulationSpace=World
    alphaBoolBlend {blendOption=Linear}
    boundBone {boneName=dyn_mail_r_4}
    componentAppliedLinearAccClamp=<1000 1000 1000>
    componentLinearAccScale=<0.1 0.1 0.1>
    componentLinearVelScale=<0.1 0.1 0.1>
    gravityOverride {z=-500}
    physicsBodyDefinitions [{
      boundBone {boneName=dyn_mail_r_4}
      boxExtents=<10 10 10>
      constraintSetup {angularLimitsMax=<0 1 0>, angularLimitsMin=<0 -1 0>, angularTarget=<0 0 0>, linearAxesMax=<0 0.2 0.2>, linearAxesMin=<0 -0.2 -0.2>}
      localJointOffset=<-0.5 0 0>
     }]
   }
   n1 = Anim Dynamics - Bone: dyn_mail_r_MSTR(ComponentPose=Local To Component(LocalPose=Use cached pose 'Basic_movement'()), Alpha=1, ExternalForce="0,0,0") {
    angularBiasOverride=0.999
    angularDampingOverride=0.9
    angularSpringConstant=5000
    bAngularSpring=true
    bOverrideAngularDamping=true
    bUseGravityOverride=true
    alphaBoolBlend {blendOption=Linear}
    boundBone {boneName=dyn_mail_r_MSTR}
    componentAppliedLinearAccClamp=<500 500 500>
    componentLinearAccScale=<0.15 0.15 0.15>
    componentLinearVelScale=<0.15 0.15 0.15>
    gravityOverride {z=-500}
    physicsBodyDefinitions [{
      boundBone {boneName=dyn_mail_r_MSTR}
      boxExtents=<7 7 7>
      constraintSetup {angularLimitsMax=<0 0.5 0.5>, angularLimitsMin=<0 -0.5 -0.5>, angularTarget=<0 0 0>, linearAxesMax=<0 1 0>, linearAxesMin=<0 -1 0>}
      localJointOffset=<-1 0 0>
     }]
   }
   n2 = Anim Dynamics - Bone: dyn_mail_r_4(ComponentPose=@n1, Alpha=1, ExternalForce="0,0,0") {
    angularBiasOverride=0.999
    angularDampingOverride=0.9
    angularSpringConstant=5000
    bAngularSpring=true
    bOverrideAngularDamping=true
    bUseGravityOverride=true
    simulationSpace=RootRelative
    alphaBoolBlend {blendOption=Linear}
    boundBone {boneName=dyn_mail_r_4}
    componentAppliedLinearAccClamp=<500 500 500>
    componentLinearAccScale=<0.15 0.15 0.15>
    componentLinearVelScale=<0.15 0.15 0.15>
    gravityOverride {z=-500}
    physicsBodyDefinitions [{
      boundBone {boneName=dyn_mail_r_4}
      boxExtents=<7 7 7>
      constraintSetup {angularLimitsMax=<0 0.5 0.5>, angularLimitsMin=<0 -0.5 -0.5>, angularTarget=<0 0 0>, linearAxesMax=<0 0.5 0.5>, linearAxesMin=<0 -0.5 -0.5>}
      localJointOffset=<-1 0 0>
     }]
   }
   n3 = Anim Dynamics - Bone: dyn_mail_r_5(ComponentPose=@n2, Alpha=1, ExternalForce="0,0,0") {
    angularBiasOverride=0.999
    angularDampingOverride=0.9
    angularSpringConstant=5000
    bAngularSpring=true
    bOverrideAngularDamping=true
    bUseGravityOverride=true
    alphaBoolBlend {blendOption=Linear}
    boundBone {boneName=dyn_mail_r_5}
    componentAppliedLinearAccClamp=<500 500 500>
    componentLinearAccScale=<0.15 0.15 0.15>
    componentLinearVelScale=<0.15 0.15 0.15>
    gravityOverride {z=-500}
    physicsBodyDefinitions [{
      boundBone {boneName=dyn_mail_r_5}
      boxExtents=<7 7 7>
      constraintSetup {angularLimitsMax=<0 0.5 0.5>, angularLimitsMin=<0 -0.5 -0.5>, angularTarget=<0 0 0>, linearAxesMax=<0 1 0.5>, linearAxesMin=<0 -1 -0.5>}
      localJointOffset=<-1 0 0>
     }]
   }
   n4 = Anim Dynamics - Bone: dyn_mail_r_3(ComponentPose=@n3, Alpha=1, ExternalForce="0,0,0") {
    angularBiasOverride=0.999
    angularDampingOverride=0.9
    angularSpringConstant=5000
    bAngularSpring=true
    bOverrideAngularDamping=true
    bUseGravityOverride=true
    alphaBoolBlend {blendOption=Linear}
    boundBone {boneName=dyn_mail_r_3}
    componentAppliedLinearAccClamp=<500 500 500>
    componentLinearAccScale=<0.15 0.15 0.15>
    componentLinearVelScale=<0.15 0.15 0.15>
    gravityOverride {z=-500}
    physicsBodyDefinitions [{
      boundBone {boneName=dyn_mail_r_3}
      boxExtents=<7 7 7>
      constraintSetup {angularLimitsMax=<0 0.5 0.5>, angularLimitsMin=<0 -0.5 -0.5>, angularTarget=<0 0 0>, linearAxesMax=<0 1 0.5>, linearAxesMin=<0 -1 -0.5>}
      localJointOffset=<-1 0 0>
     }]
   }
   n5 = Anim Dynamics - Bone: dyn_mail_r_2(ComponentPose=@n4, Alpha=1, ExternalForce="0,0,0") {
    angularBiasOverride=0.999
    angularDampingOverride=0.9
    angularSpringConstant=5000
    bAngularSpring=true
    bOverrideAngularDamping=true
    bUseGravityOverride=true
    alphaBoolBlend {blendOption=Linear}
    boundBone {boneName=dyn_mail_r_2}
    componentAppliedLinearAccClamp=<500 500 500>
    componentLinearAccScale=<0.15 0.15 0.15>
    componentLinearVelScale=<0.15 0.15 0.15>
    gravityOverride {z=-500}
    physicsBodyDefinitions [{
      boundBone {boneName=dyn_mail_r_2}
      boxExtents=<7 7 7>
      constraintSetup {angularLimitsMax=<0 0.5 0.5>, angularLimitsMin=<0 -0.5 -0.5>, angularTarget=<0 0 0>, linearAxesMax=<0 1 0.5>, linearAxesMin=<0 -1 -0.5>}
      localJointOffset=<-1 0 0>
     }]
   }
   n6 = Anim Dynamics - Bone: dyn_mail_r_1(ComponentPose=@n5, Alpha=1, ExternalForce="0,0,0") {
    angularBiasOverride=0.999
    angularDampingOverride=0.9
    angularSpringConstant=5000
    bAngularSpring=true
    bOverrideAngularDamping=true
    bUseGravityOverride=true
    alphaBoolBlend {blendOption=Linear}
    boundBone {boneName=dyn_mail_r_1}
    componentAppliedLinearAccClamp=<500 500 500>
    componentLinearAccScale=<0.15 0.15 0.15>
    componentLinearVelScale=<0.15 0.15 0.15>
    gravityOverride {z=-500}
    physicsBodyDefinitions [{
      boundBone {boneName=dyn_mail_r_1}
      boxExtents=<7 7 7>
      constraintSetup {angularLimitsMax=<0 0.5 0.5>, angularLimitsMin=<0 -0.5 -0.5>, angularTarget=<0 0 0>, linearAxesMax=<0 1 0.5>, linearAxesMin=<0 -1 -0.5>}
      localJointOffset=<-1 0 0>
     }]
   }
   n7 = Anim Dynamics - Bone: dyn_mail_r_6(ComponentPose=@n6, Alpha=1, ExternalForce="0,0,0") {
    angularBiasOverride=0.999
    angularDampingOverride=0.9
    angularSpringConstant=5000
    bAngularSpring=true
    bOverrideAngularDamping=true
    bUseGravityOverride=true
    alphaBoolBlend {blendOption=Linear}
    boundBone {boneName=dyn_mail_r_6}
    componentAppliedLinearAccClamp=<500 500 500>
    componentLinearAccScale=<0.15 0.15 0.15>
    componentLinearVelScale=<0.15 0.15 0.15>
    gravityOverride {z=-500}
    physicsBodyDefinitions [{
      boundBone {boneName=dyn_mail_r_6}
      boxExtents=<7 7 7>
      constraintSetup {angularLimitsMax=<0 0.5 0.5>, angularLimitsMin=<0 -0.5 -0.5>, angularTarget=<0 0 0>, linearAxesMax=<0 1 0.5>, linearAxesMin=<0 -1 -0.5>}
      localJointOffset=<-1 0 0>
     }]
   }
   n8 = Anim Dynamics - Bone: dyn_mail_l_MSTR(ComponentPose=@n7, Alpha=1, ExternalForce="0,0,0") {
    angularBiasOverride=0.999
    angularDampingOverride=0.9
    angularSpringConstant=5000
    bAngularSpring=true
    bOverrideAngularDamping=true
    bUseGravityOverride=true
    alphaBoolBlend {blendOption=Linear}
    boundBone {boneName=dyn_mail_l_MSTR}
    componentAppliedLinearAccClamp=<500 500 500>
    componentLinearAccScale=<0.15 0.15 0.15>
    componentLinearVelScale=<0.15 0.15 0.15>
    gravityOverride {z=-500}
    physicsBodyDefinitions [{
      boundBone {boneName=dyn_mail_l_MSTR}
      boxExtents=<7 7 7>
      constraintSetup {angularLimitsMax=<0 0.5 0.5>, angularLimitsMin=<0 -0.5 -0.5>, angularTarget=<0 0 0>, linearAxesMax=<0 1 0.5>, linearAxesMin=<0 -1 -0.5>}
      localJointOffset=<-1 0 0>
     }]
   }
   n9 = Anim Dynamics - Bone: dyn_mail_l_3(ComponentPose=@n8, Alpha=1, ExternalForce="0,0,0") {
    angularBiasOverride=0.999
    angularDampingOverride=0.9
    angularSpringConstant=5000
    bAngularSpring=true
    bOverrideAngularDamping=true
    bUseGravityOverride=true
    alphaBoolBlend {blendOption=Linear}
    boundBone {boneName=dyn_mail_l_3}
    componentAppliedLinearAccClamp=<500 500 500>
    componentLinearAccScale=<0.15 0.15 0.15>
    componentLinearVelScale=<0.15 0.15 0.15>
    gravityOverride {z=-500}
    physicsBodyDefinitions [{
      boundBone {boneName=dyn_mail_l_3}
      boxExtents=<7 7 7>
      constraintSetup {angularLimitsMax=<10 0 0>, angularLimitsMin=<-10 -10 0>, angularTarget=<0 0 0>, linearAxesMax=<0 1 0.5>, linearAxesMin=<0 -1 -0.5>}
      localJointOffset=<-1 -1 -1>
     }]
   }
   n10 = Anim Dynamics - Bone: dyn_mail_l_2(ComponentPose=@n9, Alpha=1, ExternalForce="0,0,0") {
    angularBiasOverride=0.999
    angularDampingOverride=0.9
    angularSpringConstant=5000
    bAngularSpring=true
    bOverrideAngularDamping=true
    bUseGravityOverride=true
    alphaBoolBlend {blendOption=Linear}
    boundBone {boneName=dyn_mail_l_2}
    componentAppliedLinearAccClamp=<500 500 500>
    componentLinearAccScale=<0.15 0.15 0.15>
    componentLinearVelScale=<0.15 0.15 0.15>
    gravityOverride {z=-500}
    physicsBodyDefinitions [{
      boundBone {boneName=dyn_mail_l_2}
      boxExtents=<7 7 7>
      constraintSetup {angularLimitsMax=<0 0 0>, angularLimitsMin=<0 0 0>, angularTarget=<0 0 0>, linearAxesMax=<0 1 0.5>, linearAxesMin=<0 -1 -0.5>}
      localJointOffset=<-1 -1 -1>
     }]
   }
   n11 = Anim Dynamics - Bone: dyn_mail_l_1(ComponentPose=@n10, Alpha=1, ExternalForce="0,0,0") {
    angularBiasOverride=0.999
    angularDampingOverride=0.9
    angularSpringConstant=5000
    bAngularSpring=true
    bOverrideAngularDamping=true
    bUseGravityOverride=true
    alphaBoolBlend {blendOption=Linear}
    boundBone {boneName=dyn_mail_l_1}
    componentAppliedLinearAccClamp=<500 500 500>
    componentLinearAccScale=<0.15 0.15 0.15>
    componentLinearVelScale=<0.15 0.15 0.15>
    gravityOverride {z=-500}
    physicsBodyDefinitions [{
      boundBone {boneName=dyn_mail_l_1}
      boxExtents=<7 7 7>
      constraintSetup {angularLimitsMax=<5 0 0>, angularLimitsMin=<-5 -5 0>, angularTarget=<0 0 0>, linearAxesMax=<0 1 0.5>, linearAxesMin=<0 -1 -0.5>}
      localJointOffset=<-1 -1 -1>
     }]
   }
   n12 = Anim Dynamics - Bone: dyn_mail_under_2(ComponentPose=@n11, Alpha=1, ExternalForce="0,0,0") {
    angularBiasOverride=0.999
    angularDampingOverride=0.9
    angularSpringConstant=5000
    bAngularSpring=true
    bOverrideAngularDamping=true
    bUseGravityOverride=true
    bUseSphericalLimits=true
    alphaBoolBlend {blendOption=Linear}
    boundBone {boneName=dyn_mail_under_2}
    componentAppliedLinearAccClamp=<500 500 500>
    componentLinearAccScale=<0.15 0.15 0.15>
    componentLinearVelScale=<0.15 0.15 0.15>
    gravityOverride {z=-500}
    physicsBodyDefinitions [{
      boundBone {boneName=dyn_mail_under_2}
      boxExtents=<7 7 7>
      constraintSetup {angularLimitsMax=<5 5 5>, angularLimitsMin=<-5 -5 0>, angularTarget=<0 0 0>, linearAxesMax=<0 0.5 0>, linearAxesMin=<0 -0.5 0>}
      localJointOffset=<-1 -1 -1>
     }]
    sphericalLimits [{limitRadius=1, drivingBone {boneName=dyn_trapdor}, sphereLocalOffset=<0 0 0>}]
   }
   n13 = Anim Dynamics - Bone: dyn_mail_under_1(ComponentPose=@n12, Alpha=1, ExternalForce="0,0,0") {
    angularBiasOverride=0.999
    angularDampingOverride=0.9
    angularSpringConstant=5000
    bAngularSpring=true
    bOverrideAngularDamping=true
    bUseGravityOverride=true
    alphaBoolBlend {blendOption=Linear}
    boundBone {boneName=dyn_mail_under_1}
    componentAppliedLinearAccClamp=<500 500 500>
    componentLinearAccScale=<0.15 0.15 0.15>
    componentLinearVelScale=<0.15 0.15 0.15>
    gravityOverride {z=-500}
    physicsBodyDefinitions [{
      boundBone {boneName=dyn_mail_under_1}
      boxExtents=<7 7 7>
      constraintSetup {angularLimitsMax=<5 5 5>, angularLimitsMin=<-5 -5 0>, angularTarget=<0 0 0>, linearAxesMax=<0 1 0.5>, linearAxesMin=<0 -1 -0.5>}
      localJointOffset=<-1 -1 -1>
     }]
   }
   n14 = Anim Dynamics - Bone: dyn_mail_under_5(ComponentPose=@n13, Alpha=1, ExternalForce="0,0,0") {
    angularBiasOverride=0.999
    angularDampingOverride=0.9
    angularSpringConstant=5000
    bAngularSpring=true
    bOverrideAngularDamping=true
    bUseGravityOverride=true
    alphaBoolBlend {blendOption=Linear}
    boundBone {boneName=dyn_mail_under_5}
    componentAppliedLinearAccClamp=<500 500 500>
    componentLinearAccScale=<0.15 0.15 0.15>
    componentLinearVelScale=<0.15 0.15 0.15>
    gravityOverride {z=-500}
    physicsBodyDefinitions [{
      boundBone {boneName=dyn_mail_under_5}
      boxExtents=<7 7 7>
      constraintSetup {angularLimitsMax=<0 0 0>, angularLimitsMin=<0 0 0>, angularTarget=<0 0 0>, linearAxesMax=<0 1 0.5>, linearAxesMin=<0 -1 -0.5>}
      localJointOffset=<-1 -1 -1>
     }]
   }
   n15 = Anim Dynamics - Bone: dyn_mail_under_4(ComponentPose=@n14, Alpha=1, ExternalForce="0,0,0") {
    angularBiasOverride=0.999
    angularDampingOverride=0.9
    angularSpringConstant=5000
    bAngularSpring=true
    bOverrideAngularDamping=true
    bUseGravityOverride=true
    alphaBoolBlend {blendOption=Linear}
    boundBone {boneName=dyn_mail_under_4}
    componentAppliedLinearAccClamp=<500 500 500>
    componentLinearAccScale=<0.15 0.15 0.15>
    componentLinearVelScale=<0.15 0.15 0.15>
    gravityOverride {z=-500}
    physicsBodyDefinitions [{
      boundBone {boneName=dyn_mail_under_4}
      boxExtents=<7 7 7>
      constraintSetup {angularLimitsMax=<0 0 0>, angularLimitsMin=<0 0 0>, angularTarget=<0 0 0>, linearAxesMax=<0 1 0.5>, linearAxesMin=<0 -1 -0.5>}
      localJointOffset=<-1 -1 -1>
     }]
   }
   n16 = Anim Dynamics - Bone: dyn_mail_under_3(ComponentPose=@n15, Alpha=1, ExternalForce="0,0,0") {
    angularBiasOverride=0.999
    angularDampingOverride=0.9
    angularSpringConstant=5000
    bAngularSpring=true
    bOverrideAngularDamping=true
    bUseGravityOverride=true
    bUseSphericalLimits=true
    simulationSpace=BoneRelative
    alphaBoolBlend {blendOption=Linear}
    boundBone {boneName=dyn_mail_under_3}
    componentAppliedLinearAccClamp=<500 500 500>
    componentLinearAccScale=<0.15 0.15 0.15>
    componentLinearVelScale=<0.15 0.15 0.15>
    gravityOverride {z=-500}
    physicsBodyDefinitions [{
      boundBone {boneName=dyn_mail_under_3}
      boxExtents=<7 7 7>
      constraintSetup {angularLimitsMax=<0 0 0>, angularLimitsMin=<0 0 0>, angularTarget=<0 0 0>, linearAxesMax=<0 0.5 0>, linearAxesMin=<0 -0.5 0>}
      localJointOffset=<-1 0 0>
     }]
    planarLimits [{drivingBone {boneName=dyn_trapdor}, planeTransform {location=<-0.9537 0.0152 0.0011>, rotation=<rot -3.8394 -90.9197 -86.2739>}}]
    sphericalLimits [{limitRadius=1, drivingBone {boneName=dyn_trapdor}, sphereLocalOffset=<0 0 0>}]
   }
   n17 = Anim Dynamics - Bone: dyn_mail_r_5(ComponentPose=Local To Component(LocalPose=Use cached pose 'Basic_movement'()), Alpha=1, ExternalForce="0,0,0") {
    angularBiasOverride=0.999
    angularDampingOverride=1
    angularSpringConstant=5000
    bAngularSpring=true
    bOverrideAngularBias=true
    bOverrideAngularDamping=true
    bUseGravityOverride=true
    simulationSpace=World
    alphaBoolBlend {blendOption=Linear}
    boundBone {boneName=dyn_mail_r_5}
    componentAppliedLinearAccClamp=<1000 1000 1000>
    componentLinearAccScale=<0.1 0.1 0.1>
    componentLinearVelScale=<0.1 0.1 0.1>
    gravityOverride {z=-500}
    physicsBodyDefinitions [{
      boundBone {boneName=dyn_mail_r_5}
      boxExtents=<10 10 10>
      constraintSetup {angularLimitsMax=<0 1 0>, angularLimitsMin=<0 -1 0>, angularTarget=<0 0 0>, linearAxesMax=<0 0.2 0.2>, linearAxesMin=<0 -0.2 -0.2>}
      localJointOffset=<-0.5 0 0>
     }]
   }
   n18 = Anim Dynamics - Bone: dyn_mail_r_3(ComponentPose=@n17, Alpha=1, ExternalForce="0,0,0") {
    angularBiasOverride=0.999
    angularDampingOverride=1
    angularSpringConstant=5000
    bAngularSpring=true
    bOverrideAngularBias=true
    bOverrideAngularDamping=true
    bUseGravityOverride=true
    simulationSpace=World
    alphaBoolBlend {blendOption=Linear}
    boundBone {boneName=dyn_mail_r_3}
    componentAppliedLinearAccClamp=<1000 1000 1000>
    componentLinearAccScale=<0.1 0.1 0.1>
    componentLinearVelScale=<0.1 0.1 0.1>
    gravityOverride {z=-500}
    physicsBodyDefinitions [{
      boundBone {boneName=dyn_mail_r_3}
      boxExtents=<10 10 10>
      constraintSetup {angularLimitsMax=<0 1 1>, angularLimitsMin=<0 -1 -1>, angularTarget=<0 0 0>, linearAxesMax=<0 0.2 0.2>, linearAxesMin=<0 -0.2 -0.2>}
      localJointOffset=<-0.5 0 0>
     }]
   }
   n19 = Anim Dynamics - Bone: dyn_mail_r_2(ComponentPose=@n18, Alpha=1, ExternalForce="0,0,0") {
    angularBiasOverride=0.999
    angularDampingOverride=1
    angularSpringConstant=5000
    bAngularSpring=true
    bOverrideAngularBias=true
    bOverrideAngularDamping=true
    bUseGravityOverride=true
    simulationSpace=World
    alphaBoolBlend {blendOption=Linear}
    boundBone {boneName=dyn_mail_r_2}
    componentAppliedLinearAccClamp=<1000 1000 1000>
    componentLinearAccScale=<0.1 0.1 0.1>
    componentLinearVelScale=<0.1 0.1 0.1>
    gravityOverride {z=-500}
    physicsBodyDefinitions [{
      boundBone {boneName=dyn_mail_r_2}
      boxExtents=<10 10 10>
      constraintSetup {angularLimitsMax=<0 1 1>, angularLimitsMin=<0 -1 -1>, angularTarget=<0 0 0>, linearAxesMax=<0 0.2 0.2>, linearAxesMin=<0 -0.2 -0.2>}
      localJointOffset=<-0.5 0 0>
     }]
   }
   n20 = Anim Dynamics - Bone: dyn_mail_r_1(ComponentPose=@n19, Alpha=1, ExternalForce="0,0,0") {
    angularBiasOverride=0.999
    angularDampingOverride=1
    angularSpringConstant=5000
    bAngularSpring=true
    bOverrideAngularBias=true
    bOverrideAngularDamping=true
    bUseGravityOverride=true
    simulationSpace=World
    alphaBoolBlend {blendOption=Linear}
    boundBone {boneName=dyn_mail_r_1}
    componentAppliedLinearAccClamp=<1000 1000 1000>
    componentLinearAccScale=<0.1 0.1 0.1>
    componentLinearVelScale=<0.1 0.1 0.1>
    gravityOverride {z=-500}
    physicsBodyDefinitions [{
      boundBone {boneName=dyn_mail_r_1}
      boxExtents=<10 10 10>
      constraintSetup {angularLimitsMax=<0 1 1>, angularLimitsMin=<0 -1 -1>, angularTarget=<0 0 0>, linearAxesMax=<0 0.2 0.2>, linearAxesMin=<0 -0.2 -0.2>}
      localJointOffset=<-0.5 0 0>
     }]
   }
   n21 = Anim Dynamics - Bone: dyn_mail_r_6(ComponentPose=@n20, Alpha=1, ExternalForce="0,0,0") {
    angularBiasOverride=0.999
    angularDampingOverride=1
    angularSpringConstant=5000
    bAngularSpring=true
    bOverrideAngularBias=true
    bOverrideAngularDamping=true
    bUseGravityOverride=true
    simulationSpace=World
    alphaBoolBlend {blendOption=Linear}
    boundBone {boneName=dyn_mail_r_6}
    componentAppliedLinearAccClamp=<1000 1000 1000>
    componentLinearAccScale=<0.1 0.1 0.1>
    componentLinearVelScale=<0.1 0.1 0.1>
    gravityOverride {z=-500}
    physicsBodyDefinitions [{
      boundBone {boneName=dyn_mail_r_6}
      boxExtents=<10 10 10>
      constraintSetup {angularLimitsMax=<0 1 1>, angularLimitsMin=<0 -1 -1>, angularTarget=<0 0 0>, linearAxesMax=<0 0.2 0.2>, linearAxesMin=<0 -0.2 -0.2>}
      localJointOffset=<-0.5 0 0>
     }]
   }
   n22 = Anim Dynamics - Bone: dyn_mail_l_3(ComponentPose=@n21, Alpha=1, ExternalForce="0,0,0") {
    angularBiasOverride=0.999
    angularDampingOverride=1
    angularSpringConstant=5000
    bAngularSpring=true
    bOverrideAngularBias=true
    bOverrideAngularDamping=true
    bUseGravityOverride=true
    simulationSpace=World
    alphaBoolBlend {blendOption=Linear}
    boundBone {boneName=dyn_mail_l_3}
    componentAppliedLinearAccClamp=<1000 1000 1000>
    componentLinearAccScale=<0.1 0.1 0.1>
    componentLinearVelScale=<0.1 0.1 0.1>
    gravityOverride {z=-500}
    physicsBodyDefinitions [{
      boundBone {boneName=dyn_mail_l_3}
      boxExtents=<10 10 10>
      constraintSetup {angularLimitsMax=<0 1 1>, angularLimitsMin=<0 -1 -1>, angularTarget=<0 0 0>, linearAxesMax=<0 0.2 0.2>, linearAxesMin=<0 -0.2 -0.2>}
      localJointOffset=<-0.5 0 -1>
     }]
   }
   n23 = Anim Dynamics - Bone: dyn_mail_l_2(ComponentPose=@n22, Alpha=1, ExternalForce="0,0,0") {
    angularBiasOverride=0.999
    angularDampingOverride=1
    angularSpringConstant=5000
    bAngularSpring=true
    bOverrideAngularBias=true
    bOverrideAngularDamping=true
    bUseGravityOverride=true
    simulationSpace=World
    alphaBoolBlend {blendOption=Linear}
    boundBone {boneName=dyn_mail_l_2}
    componentAppliedLinearAccClamp=<1000 1000 1000>
    componentLinearAccScale=<0.1 0.1 0.1>
    componentLinearVelScale=<0.1 0.1 0.1>
    gravityOverride {z=-500}
    physicsBodyDefinitions [{
      boundBone {boneName=dyn_mail_l_2}
      boxExtents=<10 10 10>
      constraintSetup {angularLimitsMax=<0 1 1>, angularLimitsMin=<0 -1 -1>, angularTarget=<0 0 0>, linearAxesMax=<0 0.2 0.2>, linearAxesMin=<0 -0.2 -0.2>}
      localJointOffset=<-0.5 0 -1>
     }]
   }
   n24 = Anim Dynamics - Bone: dyn_mail_l_1(ComponentPose=@n23, Alpha=1, ExternalForce="0,0,0") {
    angularBiasOverride=0.999
    angularDampingOverride=1
    angularSpringConstant=5000
    bAngularSpring=true
    bOverrideAngularBias=true
    bOverrideAngularDamping=true
    bUseGravityOverride=true
    simulationSpace=World
    alphaBoolBlend {blendOption=Linear}
    boundBone {boneName=dyn_mail_l_1}
    componentAppliedLinearAccClamp=<1000 1000 1000>
    componentLinearAccScale=<0.1 0.1 0.1>
    componentLinearVelScale=<0.1 0.1 0.1>
    gravityOverride {z=-500}
    physicsBodyDefinitions [{
      boundBone {boneName=dyn_mail_l_1}
      boxExtents=<10 10 10>
      constraintSetup {angularLimitsMax=<0 1 1>, angularLimitsMin=<0 -1 -1>, angularTarget=<0 0 0>, linearAxesMax=<0 0.2 0.2>, linearAxesMin=<0 -0.2 -0.2>}
      localJointOffset=<-0.5 0 -1>
     }]
   }
   n25 = Anim Dynamics - Bone: dyn_mail_under_2(ComponentPose=@n24, Alpha=1, ExternalForce="0,0,0") {
    angularBiasOverride=0.999
    angularDampingOverride=1
    angularSpringConstant=5000
    bAngularSpring=true
    bOverrideAngularBias=true
    bOverrideAngularDamping=true
    bUseGravityOverride=true
    bUseSphericalLimits=true
    simulationSpace=World
    alphaBoolBlend {blendOption=Linear}
    boundBone {boneName=dyn_mail_under_2}
    componentAppliedLinearAccClamp=<500 500 500>
    componentLinearAccScale=<0.1 0.1 0.1>
    componentLinearVelScale=<0.1 0.1 0.1>
    gravityOverride {z=-500}
    physicsBodyDefinitions [{
      boundBone {boneName=dyn_mail_under_2}
      boxExtents=<10 10 10>
      constraintSetup {angularLimitsMax=<0 0 0>, angularLimitsMin=<0 0 0>, angularTarget=<0 0 0>, linearAxesMax=<0 0.1 0>, linearAxesMin=<0 -0.1 0>}
      localJointOffset=<-0.5 0 0>
     }]
    sphericalLimits [{limitRadius=1, drivingBone {boneName=dyn_trapdor}, sphereLocalOffset=<0 0 0>}]
   }
   n26 = Anim Dynamics - Bone: dyn_mail_under_1(ComponentPose=@n25, Alpha=1, ExternalForce="0,0,0") {
    angularBiasOverride=0.999
    angularDampingOverride=1
    angularSpringConstant=5000
    bAngularSpring=true
    bOverrideAngularBias=true
    bOverrideAngularDamping=true
    bUseGravityOverride=true
    simulationSpace=World
    alphaBoolBlend {blendOption=Linear}
    boundBone {boneName=dyn_mail_under_1}
    componentAppliedLinearAccClamp=<1000 1000 1000>
    componentLinearAccScale=<0.1 0.1 0.1>
    componentLinearVelScale=<0.1 0.1 0.1>
    gravityOverride {z=-500}
    physicsBodyDefinitions [{
      boundBone {boneName=dyn_mail_under_1}
      boxExtents=<10 10 10>
      constraintSetup {angularLimitsMax=<0 1 1>, angularLimitsMin=<0 -1 -1>, angularTarget=<0 0 0>, linearAxesMax=<0 0.2 0.2>, linearAxesMin=<0 -0.2 -0.2>}
      localJointOffset=<-0.5 0 -1>
     }]
   }
   n27 = Anim Dynamics - Bone: dyn_mail_under_5(ComponentPose=@n26, Alpha=1, ExternalForce="0,0,0") {
    angularBiasOverride=0.999
    angularDampingOverride=1
    angularSpringConstant=5000
    bAngularSpring=true
    bOverrideAngularBias=true
    bOverrideAngularDamping=true
    bUseGravityOverride=true
    simulationSpace=World
    alphaBoolBlend {blendOption=Linear}
    boundBone {boneName=dyn_mail_under_5}
    componentAppliedLinearAccClamp=<1000 1000 1000>
    componentLinearAccScale=<0.1 0.1 0.1>
    componentLinearVelScale=<0.1 0.1 0.1>
    gravityOverride {z=-500}
    physicsBodyDefinitions [{
      boundBone {boneName=dyn_mail_under_5}
      boxExtents=<10 10 10>
      constraintSetup {angularLimitsMax=<0 1 1>, angularLimitsMin=<0 -1 -1>, angularTarget=<0 0 0>, linearAxesMax=<0 0.2 0.2>, linearAxesMin=<0 -0.2 -0.2>}
      localJointOffset=<-0.5 0 -1>
     }]
   }
   n28 = Anim Dynamics - Bone: dyn_mail_under_4(ComponentPose=@n27, Alpha=1, ExternalForce="0,0,0") {
    angularBiasOverride=0.999
    angularDampingOverride=1
    angularSpringConstant=5000
    bAngularSpring=true
    bOverrideAngularBias=true
    bOverrideAngularDamping=true
    bUseGravityOverride=true
    simulationSpace=World
    alphaBoolBlend {blendOption=Linear}
    boundBone {boneName=dyn_mail_under_4}
    componentAppliedLinearAccClamp=<1000 1000 1000>
    componentLinearAccScale=<0.1 0.1 0.1>
    componentLinearVelScale=<0.1 0.1 0.1>
    gravityOverride {z=-500}
    physicsBodyDefinitions [{
      boundBone {boneName=dyn_mail_under_4}
      boxExtents=<10 10 10>
      constraintSetup {angularLimitsMax=<0 1 1>, angularLimitsMin=<0 -1 -1>, angularTarget=<0 0 0>, linearAxesMax=<0 0.2 0.2>, linearAxesMin=<0 -0.2 -0.2>}
      localJointOffset=<-0.5 0 -1>
     }]
   }
   n29 = Anim Dynamics - Bone: dyn_mail_under_3(ComponentPose=@n28, Alpha=1, ExternalForce="0,0,0") {
    angularBiasOverride=0.999
    angularDampingOverride=1
    angularSpringConstant=5000
    bAngularSpring=true
    bOverrideAngularBias=true
    bOverrideAngularDamping=true
    bUseGravityOverride=true
    bUseSphericalLimits=true
    simulationSpace=World
    alphaBoolBlend {blendOption=Linear}
    boundBone {boneName=dyn_mail_under_3}
    componentAppliedLinearAccClamp=<1000 1000 1000>
    componentLinearAccScale=<0.1 0.1 0.1>
    componentLinearVelScale=<0.1 0.1 0.1>
    gravityOverride {z=-500}
    physicsBodyDefinitions [{
      boundBone {boneName=dyn_mail_under_3}
      boxExtents=<10 10 10>
      constraintSetup {angularLimitsMax=<0 1 1>, angularLimitsMin=<0 -1 -1>, angularTarget=<0 0 0>, linearAxesMax=<0 0.2 0>, linearAxesMin=<0 -0.2 0>}
      localJointOffset=<-0.5 0 0>
     }]
    planarLimits [{drivingBone {boneName=dyn_trapdor}, planeTransform {location=<-0.9537 0.0152 0.0011>, rotation=<rot -3.8394 -90.9197 -86.274>}}]
    sphericalLimits [{limitRadius=1, drivingBone {boneName=dyn_trapdor}, sphereLocalOffset=<0 0 0>}]
   }
   Output Pose(Result=Blend Poses by bool(BlendPose_0=Component To Local(ComponentPose=@n16), BlendPose_1=Component To Local(ComponentPose=@n29)) {blendType=Linear}) {layerGroup=None}
   AnimGraphNode_SubInput_0 = Input Pose()
   n30 = Anim Dynamics - Bone: dyn_trapdor(ComponentPose=Local To Component(LocalPose=@AnimGraphNode_SubInput_0), Alpha=1, ExternalForce="0,0,0") {
    angularBiasOverride=0.999
    angularDampingOverride=0.99
    angularSpringConstant=200
    bAngularSpring=true
    bOverrideAngularDamping=true
    bUseGravityOverride=true
    simulationSpace=RootRelative
    alphaBoolBlend {blendOption=Linear}
    boundBone {boneName=dyn_trapdor}
    componentAppliedLinearAccClamp=<1000 1000 1000>
    componentLinearAccScale=<0.3 0.2 0.2>
    componentLinearVelScale=<0.3 0.2 0.2>
    gravityOverride {z=-500}
    physicsBodyDefinitions [{
      boundBone {boneName=dyn_trapdor}
      boxExtents=<15 15 15>
      constraintSetup {angularLimitsMax=<0 0 100>, angularLimitsMin=<0 0 -10>, angularTarget=<0 0 0>}
      localJointOffset=<2.5 2.5 0>
     }]
   }
   n31 = Control Rig(Source=Component To Local(ComponentPose=@n30)) {
    alphaBoolBlend {blendOption=Linear}
    controlRigAssetReference {blueprintRigClass=&/LiverRomaine_Axe_CtrlRig.LiverRomaine_Axe_CtrlRig_C}
   }
   n32 = "Transform (Modify) Bone - Bone: dyn_mail_r_4"(ComponentPose=Local To Component(LocalPose=@n31), Alpha=1, Translation="-0.429,0.000,0.014", Rotation="-1.776,-0.121,-0.100", Scale="1,1,1") {
    rotationMode=BMM_Additive
    rotationSpace=BCS_BoneSpace
    translationMode=BMM_Additive
    translationSpace=BCS_BoneSpace
    alphaBoolBlend {blendOption=Linear}
    boneToModify {boneName=dyn_mail_r_4}
    rotation=<rot -1.776 -0.121 -0.1>
    translation {x=-0.429, z=0.014}
   }
   n33 = "Transform (Modify) Bone - Bone: dyn_mail_r_3"(ComponentPose=@n32, Alpha=1, Translation="-0.353,0.008,0.020", Rotation="-4.300,1.253,-0.171", Scale="1,1,1") {
    rotationMode=BMM_Additive
    rotationSpace=BCS_BoneSpace
    translationMode=BMM_Additive
    translationSpace=BCS_BoneSpace
    alphaBoolBlend {blendOption=Linear}
    boneToModify {boneName=dyn_mail_r_3}
    rotation=<rot -4.3 1.2531 -0.1711>
    translation=<-0.353 0.008 0.02>
   }
   n34 = "Transform (Modify) Bone - Bone: dyn_mail_l_1"(ComponentPose=@n33, Alpha=1, Translation="-0.353,0.006,-0.281", Rotation="-1.011,0.989,0.950", Scale="1,1,1") {
    rotationMode=BMM_Additive
    rotationSpace=BCS_BoneSpace
    translationMode=BMM_Additive
    translationSpace=BCS_BoneSpace
    alphaBoolBlend {blendOption=Linear}
    boneToModify {boneName=dyn_mail_l_1}
    rotation=<rot -1.011 0.989 0.95>
    translation=<-0.353 0.006 -0.281>
   }
   Save cached pose 'Basic_movement'(Pose=Component To Local(ComponentPose=@n34)) {cacheName=Basic_movement}
  ]}
 EventGraph {nodes [
   Event BlueprintUpdateAnimation()
   > n1 = Get!(execute:Exec) -> (then:Exec, else:Exec, BodyMesh:Skeletal Mesh Component Budgeted Object Reference)
   > Set Character_Holding_Pickaxe(Character_Holding_Pickaxe:bool=true) -> (Output_Get:bool)
   Set Character_Holding_Pickaxe(execute=@n1.else, Character_Holding_Pickaxe=false)
   n2 = Event OnCharacterCustomizationLoaded()
   > Set BodyMesh(BodyMesh:Skeletal Mesh Component Budgeted Object Reference=Get Array Item with EFort Custom Part Type(Array=@n2.CustomMeshParts)) -> (Output_Get:Skeletal Mesh Component Budgeted Object Reference)
   TryGetPawnOwner()
  ]}
]
variables [
 BodyMesh {type=SkeletalMeshComponentBudgeted}
 Character_Holding_Pickaxe {type=boolean}
]
